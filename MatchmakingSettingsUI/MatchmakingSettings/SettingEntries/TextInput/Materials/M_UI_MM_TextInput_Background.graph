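# UI-domain, alpha-composite material for the text input background in matchmaking settings.
# MF_UI_MM_OverlayBackplates supplies the B Weak color and opacity, each multiplied by a pixel-radius rounded-corner mask and vertex alpha.
# An IsDisabled parameter drives lerps on color and opacity, but both inputs of each lerp are identical, so it currently has no visible effect.
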Material /MatchmakingSettingsUI/MatchmakingSettings/SettingEntries/TextInput/Materials/M_UI_MM_TextInput_Background
graph {
 properties {bAutomaticallySetUsageInEditor=false, blendMode=BLEND_AlphaComposite, materialDomain=MD_UI, usedShadingModels=Default Lit}
 nodes [
  n1 = MaterialFunctionCall() {materialFunction=/MatchmakingSettingsUI/MatchmakingSettings/Materials/MF_UI_MM_OverlayBackplates}
  n2 = ScalarParameter() {parameterName=IsDisabled, sliderMax=1}
  n3 = Multiply(
   A=MaterialFunctionCall(Corner Radius=Constant() {r=10}, Radius in Pixels=Constant() {r=1}) {materialFunction=&MF_UI_CoreUI_RoundedMask}.Shape Mask
   B=VertexColor().A
  )
  Multiply(A=LinearInterpolate(A=@n1.B Weak Color, B=@n1.B Weak Color, Alpha=@n2), B=@n3)
  Multiply(A=LinearInterpolate(A=@n1.B Weak Opacity, B=@n1.B Weak Opacity, Alpha=@n2), B=@n3)
 ]
}
