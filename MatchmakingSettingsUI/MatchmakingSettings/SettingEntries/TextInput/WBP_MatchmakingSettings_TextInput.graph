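# Text input field widget for matchmaking settings entries, pairing an editable text box with a character-count badge over a material background.
# PreConstruct runs SetMobileStyle, which picks the minimum-size spacer, badge size and font size based on IsUsingCompactUI, keeping other font attributes intact.
# SetCharacterCount writes an integer into the badge and collapses it unless the value is above zero.
# The graphs contain no text-change binding, so the count is supplied by a caller.
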WidgetBlueprint /MatchmakingSettingsUI/MatchmakingSettings/SettingEntries/TextInput/WBP_MatchmakingSettings_TextInput
graphs [
 EventGraph {nodes [
   Event PreConstruct()
   > SetMobileStyle() {is=title K2Node_CallFunction}
  ]}
 SetCharacterCount {nodes [
   n1 = SetCharacterCount!() -> (then, Value:int) {is=title K2Node_FunctionEntry}
   > SetBadgeText!(
    execute
    self:WBP UIKit Badge Base Object Reference=Get Badge_CharacterCount() -> (Badge_CharacterCount:WBP UIKit Badge Object Reference)
    Text:Text="To Text (Integer)"(Value=@n1)
   ) -> (then)
   > Set Visibility(Visibility=Select(Option 0=Collapsed, Option 1=SelfHitTestInvisible, Index="integer > integer"(A=@n1)), self=Get Badge_CharacterCount())
  ]}
 SetMobileStyle {nodes [
   SetMobileStyle() {is=title K2Node_FunctionEntry}
   > Set Size!(
    execute
    Size:Vector 2D Structure=Select(Option 0:Vector 2D Structure=(X=400,Y=48), Option 1:Vector 2D Structure=(X=424,Y=56), Index=IsUsingCompactUI()) -> (ReturnValue:Vector 2D Structure)
    self:Spacer Object Reference=Get Min_Size() -> (Min_Size:Spacer Object Reference)
   ) -> (then, Output_Get:Vector 2D Structure)
   > SetBadgeSize!(
    execute
    self:WBP UIKit Badge Base Object Reference=Get Badge_CharacterCount()
    BadgeSize:E_UI_BadgeSize Enum=Select(Option 0:E_UI_BadgeSize Enum=NewEnumerator0, Option 1:E_UI_BadgeSize Enum=NewEnumerator1, Index=IsUsingCompactUI()) -> (ReturnValue:E_UI_BadgeSize Enum)
   ) -> (then)
   n1 = Get EditableText() -> (EditableText:Editable Text Object Reference)
   n2 = GetFont!(self:Editable Text Object Reference=@n1) -> (
    ReturnValue_FontObject:Object Reference
    ReturnValue_FontMaterial:Object Reference
    ReturnValue_OutlineSettings:Font Outline Settings Structure
    ReturnValue_TypefaceFontName:Name
    ReturnValue_Size:float=~
    ReturnValue_LetterSpacing:int
    ReturnValue_SkewAmount:float=~
    ReturnValue_bForceMonospaced:bool=~
    ReturnValue_bMaterialIsStencil:bool=~
    ReturnValue_MonospacedWidth:float=~
   )
   > SetFont!(
    execute
    self:Editable Text Object Reference=@n1
    InFontInfo_FontObject:Object Reference=@n2
    InFontInfo_FontMaterial:Object Reference=@n2.ReturnValue_FontMaterial
    InFontInfo_OutlineSettings:Font Outline Settings Structure=@n2.ReturnValue_OutlineSettings
    InFontInfo_TypefaceFontName:Name=@n2.ReturnValue_TypefaceFontName
    InFontInfo_Size:float=Select(Option 0:float=21, Option 1:float=24, Index=IsUsingCompactUI()) -> (ReturnValue:float)
    InFontInfo_LetterSpacing:int=@n2.ReturnValue_LetterSpacing
    InFontInfo_SkewAmount:float=@n2.ReturnValue_SkewAmount
    InFontInfo_bForceMonospaced:bool=@n2.ReturnValue_bForceMonospaced
    InFontInfo_bMaterialIsStencil:bool=@n2.ReturnValue_bMaterialIsStencil
    InFontInfo_MonospacedWidth:float=@n2.ReturnValue_MonospacedWidth
   ) -> (then)
  ]}
]
variables []
widgets {
 info {inheritedWidgetCount=0, parentClass=&CommonUserWidget, rootWidgetClass=&Overlay, widgetCount=6}
 widgets [
  Overlay_27 {widgetClassPath=&Overlay, widget=Overlay_27}
  Background {
   widgetClassPath=&Image
   parent=Overlay_27
   widget=Background
   properties {
    visibility=SelfHitTestInvisible
    brush {
     bIsDynamicallyLoaded=false
     drawAs=Image
     imageType=FullColor
     mirroring=NoMirror
     resourceName=None
     resourceObject=/MatchmakingSettingsUI/MatchmakingSettings/SettingEntries/TextInput/MaterialInstances/MI_UI_MM_TextInput_Background
     tiling=NoTile
     imageSize=<32 32>
     margin=<ltrb 0 0 0 0>
     outlineSettings {
      bUseBrushTransparency=false
      roundingType=HalfHeightRadius
      width=0
      color {colorUseRule=UseColor_Specified, specifiedColor=<rgba 0 0 0 0>}
      cornerRadii=<0 0 0 0>
     }
     tintColor {colorUseRule=UseColor_Specified, specifiedColor=<rgba 1 1 1 1>}
     uVRegion {bIsValid=false, max=<0 0>, min=<0 0>}
    }
   }
   slot {class=OverlaySlot, properties {horizontalAlignment=HAlign_Fill, verticalAlignment=VAlign_Fill}}
  }
  Min_Size {
   widgetClassPath=&Spacer
   bIsVariable=true
   parent=Overlay_27
   widget=Min_Size
   properties {size=<400 48>}
   slot {class=OverlaySlot, properties {horizontalAlignment=HAlign_Fill, verticalAlignment=VAlign_Fill}}
  }
  HorizontalBox_0 {
   widgetClassPath=&HorizontalBox
   parent=Overlay_27
   widget=HorizontalBox_0
   slot {class=OverlaySlot, properties {horizontalAlignment=HAlign_Fill, verticalAlignment=VAlign_Center, padding=<ltrb 12 8 12 8>}}
  }
  EditableText {
   widgetClassPath=/Script/UMG.EditableText
   bIsVariable=true
   parent=HorizontalBox_0
   widget=EditableText
   properties {
    hintText=Input Text
    widgetStyle {
     backgroundImageComposing {
      bIsDynamicallyLoaded=false
      drawAs=Border
      imageType=FullColor
      mirroring=NoMirror
      resourceName=../../../Engine/Content/Slate/Old/HyperlinkDotted.png
      resourceObject=null
      tiling=Both
      imageSize=<0 0>
      margin=<ltrb 0 0 0 0.1875>
      outlineSettings {
       bUseBrushTransparency=false
       roundingType=HalfHeightRadius
       width=0
       color {colorUseRule=UseColor_Specified, specifiedColor=<rgba 0 0 0 0>}
       cornerRadii=<0 0 0 0>
      }
      tintColor {colorUseRule=UseColor_Specified, specifiedColor=<rgba 1 1 1 1>}
      uVRegion {bIsValid=false, max=<0 0>, min=<0 0>}
     }
     backgroundImageSelected {
      bIsDynamicallyLoaded=false
      drawAs=Box
      imageType=FullColor
      mirroring=NoMirror
      resourceName=../../../Engine/Content/Slate/Common/EditableTextSelectionBackground.png
      resourceObject=null
      tiling=NoTile
      imageSize=<0 0>
      margin=<ltrb 0.25 0.25 0.25 0.25>
      outlineSettings {
       bUseBrushTransparency=false
       roundingType=HalfHeightRadius
       width=0
       color {colorUseRule=UseColor_Specified, specifiedColor=<rgba 0 0 0 0>}
       cornerRadii=<0 0 0 0>
      }
      tintColor {colorUseRule=UseColor_Specified, specifiedColor=<rgba 1 1 1 1>}
      uVRegion {bIsValid=false, max=<0 0>, min=<0 0>}
     }
     caretImage {
      bIsDynamicallyLoaded=false
      drawAs=Image
      imageType=FullColor
      mirroring=NoMirror
      resourceName=../../../Engine/Content/Slate/Old/White.png
      resourceObject=null
      tiling=NoTile
      imageSize=<16 16>
      margin=<ltrb 0 0 0 0>
      outlineSettings {
       bUseBrushTransparency=false
       roundingType=HalfHeightRadius
       width=0
       color {colorUseRule=UseColor_Specified, specifiedColor=<rgba 0 0 0 0>}
       cornerRadii=<0 0 0 0>
      }
      tintColor {colorUseRule=UseColor_Specified, specifiedColor=<rgba 1 1 1 1>}
      uVRegion {bIsValid=false, max=<0 0>, min=<0 0>}
     }
     colorAndOpacity {colorUseRule=UseColor_Foreground, specifiedColor=<rgba 1 0 1 1>}
     font {
      bForceMonospaced=false
      bMaterialIsStencil=false
      fontMaterial=null
      fontObject=&HeadingNow
      letterSpacing=10
      monospacedWidth=1
      size=15.75
      skewAmount=0
      typefaceFontName=74 Regular
      outlineSettings {
       bApplyOutlineToDropShadows=false
       bMiteredCorners=false
       bSeparateFillAlpha=false
       outlineMaterial=null
       outlineSize=0
       outlineColor=<rgba 0 0 0 1>
      }
     }
     hintColorAndOpacity {colorUseRule=UseColor_Foreground, specifiedColor=<rgba 1 0 1 1>}
     selectedBackgroundColor {colorUseRule=UseColor_Foreground, specifiedColor=<rgba 1 0 1 1>}
    }
   }
   slot {class=HorizontalBoxSlot, properties {verticalAlignment=VAlign_Center, size {sizeRule=Fill, value=1}}}
  }
  Badge_CharacterCount {
   widgetClassPath=/FortUILibrary/UIKit/Badge/WBP_UIKit_Badge
   bIsVariable=true
   parent=HorizontalBox_0
   widget=Badge_CharacterCount
   properties {badgeText=XX, badgeType=Text, indicatorConfig=None}
   slot {class=HorizontalBoxSlot, properties {horizontalAlignment=HAlign_Left, verticalAlignment=VAlign_Center, padding=<ltrb 8 0 0 0>}}
  }
 ]
}
