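# WBP_MatchmakingSettings_TextInput_Entry is a Widget Blueprint for the text-input variant of a matchmaking settings entry.
# It derives from WBP_MatchmakingSettings_EntryBase and has no root widget of its own.
# Its only widget is a WBP_MatchmakingSettings_TextInput exposed as the variable TextInput, apparently filling a named slot of the base.
# The event graph is empty and no variables are declared, so any behavior comes from the parent class or the input widget.
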WidgetBlueprint /MatchmakingSettingsUI/MatchmakingSettings/SettingEntries/TextInput/WBP_MatchmakingSettings_TextInput_Entry
graphs [EventGraph {links [], nodes []}]
variables []
widgets {
 info {
  inheritedWidgetCount=0
  namedSlotCount=1
  parentClass=/MatchmakingSettingsUI/MatchmakingSettings/SettingEntries/WBP_MatchmakingSettings_EntryBase.WBP_MatchmakingSettings_EntryBase_C
  rootWidgetClass=null
  widgetCount=1
 }
 widgets [TextInput {
   widgetClassPath=&/WBP_MatchmakingSettings_TextInput
   bIsVariable=true
   widget=TextInput
  }]
}
